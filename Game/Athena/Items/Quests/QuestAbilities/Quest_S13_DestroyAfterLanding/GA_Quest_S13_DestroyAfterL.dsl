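; Gameplay ability for the Season 13 "destroy after landing" quest that tags its owner when triggered by a gameplay event.
; On activation from the event it calls the parent implementation, applies the GE_Quest_DestroyAfterLanding60_Tag effect to the owner, then immediately ends.
; The ability holds no timing or destruction logic of its own, only applying that effect.
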(event Ability|EventActivateAbilityFromEvent
  (params (EventData :type Gameplay%20Event%20Data%20Structure%20%28by%20ref%29))
  (|CallParentFunction EventData)
  (Ability|ApplyGameplayEffectToOwner "/Game/Athena/Items/Quests/QuestAbilities/Quest_S13_DestroyAfterLanding/GE_Quest_DestroyAfterLanding60_Tag.GE_Quest_DestroyAfterLanding60_Tag_C")
  (Ability|EndAbility))
